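# Material function that overlays a glowing cracked-ice decal, placed in UV1, onto incoming material attributes.
# A radial mask fades the crack normal, blended over the base normal, and the tinted crack emissive; a second radial mask darkens the emissive center.
# Temporal-dithered BumpOffset adds a deep crack layer for fake depth, reduced to one plain sample on low quality, mobile and forward.
# Several leftover nodes are unconnected.
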MaterialFunction /Game/Athena/WinterArt/Materials/MF_DecalIceCracked
graph {nodes [
  n1 = GetMaterialAttributes(None=FunctionInput() {inputType=FunctionInput_MaterialAttributes}) {attributeGetTypes [#1, #2, #3]}
  n2 = TextureCoordinate() {coordinateIndex=1}
  n3 = ScalarParameter() {defaultValue=3, group=CrackDecal, parameterName=CrackTiling}
  n4 = AppendVector(A=@n3, B=@n3)
  n5 = ComponentMask(None=VectorParameter() {parameterName=CrackLock, defaultValue=<rgba 0.519 0.4714 0 1>}.RGB) {g=true, r=true}
  n6 = Add(A=Multiply(A=@n2, B=@n4), B=@n5)
  n7 = TextureSampleParameter2D(UVs=@n6) {group=CrackDecal, parameterName=CrackNormal, samplerType=SAMPLERTYPE_Normal, texture=/Game/Athena/Environments/Textures/Crater_Small_N}
  n8 = Distance(A=@n6, B=Constant2Vector() {g=0.5, r=0.5})
  n9 = ScalarParameter() {defaultValue=2, group=CrackDecal, parameterName=CrackRadius}
  n10 = OneMinus(None=Saturate(None=Divide(A=@n8, B=@n9)))
  n11 = MaterialFunctionCall(
   Normal=@n7.RGB
   Flatness=LinearInterpolate(A=Constant() {r=1}, B=ScalarParameter() {defaultValue=-5, group=CrackDecal, parameterName=CrackNormal}, Alpha=@n10)
  ) {materialFunction=&FlattenNormal}
  n12 = ScalarParameter() {defaultValue=-0.02, group=CrackDecal, parameterName=CrackRadius2offset}
  n13 = Saturate(None=Divide(A=Add(A=@n8, B=@n12), B=ScalarParameter() {defaultValue=0.5, group=CrackDecal, parameterName=CrackRadius2}))
  n14 = Constant()
  n15 = TextureSampleParameter2D(UVs=@n6) {group=CrackDecal, parameterName=CrackTexture, texture=/Game/Effects/Fort_Effects/Textures/Tile/T_Cracked_Ice_D}
  n16 = BumpOffset(
   Coordinate=@n6
   Height=LinearInterpolate(
    A=ScalarParameter() {group=CrackDecal, parameterName=CarckBumpOffset_1}
    B=ScalarParameter() {defaultValue=-0.9895, group=CrackDecal, parameterName=CrackBumpOffset_2}
    Alpha=MaterialFunctionCall() {materialFunction=&DitherTemporalAA}.Result
   )
  )
  n17 = Add(
   A=Multiply(A=@n15.G, B=ScalarParameter() {defaultValue=850, group=CrackDecal, parameterName=CrackBirghtness})
   B=Multiply(
    A=TextureSampleParameter2D(UVs=@n16) {parameterName=CrackTexture, texture=&T_Cracked_Ice_D}.G
    B=ScalarParameter() {defaultValue=850, group=CrackDecal, parameterName=CrackBirghtnessdeep}
   )
  )
  n18 = ShadingPathSwitch(Default=@n17, Deferred=@n17, Forward=@n15.G, Mobile=@n15.G)
  n19 = QualitySwitch(Default=@n18, Low=@n15.G, High=@n18, Medium=@n18)
  n20 = Add(
   A=@n1.EmissiveColor
   B=Multiply(A=LinearInterpolate(A=@n14, B=@n19, Alpha=@n10), B=VectorParameter() {parameterName=ColorCracks, defaultValue=<rgba 0.272 0.5884 1 1>}.RGB)
  )
  FunctionOutput(None=SetMaterialAttributes!(
    MaterialAttributes=@n1.MaterialAttributes
    Normal=MaterialFunctionCall(BaseNormal=@n1.Normal, AdditionalNormal=@n11.Result) {materialFunction=&BlendAngleCorrectedNormals}.Result
    EmissiveColor=Multiply(A=@n13, B=@n20)
    BaseColor=@n1.MaterialAttributes
   ) {attributeSetTypes [#2, #1, #3]})
  MaterialFunctionCall(UVs=@n6, Texture Scale=@n4) {materialFunction=&ScaleUVsByCenter}
  n21 = Constant() {r=0.5}
  n22 = ScalarParameter() {defaultValue=0.5, group=CrackDecal, parameterName=CrackTiling2}
  n23 = Add(A=Multiply(A=@n2, B=AppendVector(A=@n22, B=@n22)), B=@n5)
  n24 = MaterialFunctionCall(Float2=@n23) {materialFunction=&BreakOutFloat2Components}
  Multiply(A=SphereMask(A=@n21, B=@n24.R) {attenuationRadius=0.5}, B=SphereMask(A=@n24.G, B=@n21) {attenuationRadius=0.5})
  TextureSampleParameter2D(UVs=@n23) {parameterName=CrackNormal, samplerType=SAMPLERTYPE_Normal, texture=&Crater_Small_N}
  TextureSampleParameter2D(UVs=@n6) {parameterName=CrackTexture, texture=&T_Cracked_Ice_D}
 ]}
