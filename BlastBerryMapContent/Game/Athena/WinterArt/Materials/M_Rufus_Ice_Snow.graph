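# Subsurface ice material with optional snow cover for winter environment meshes, supporting Nanite, instanced and skeletal meshes.
# The ice fakes interior glow and depth in emissive using an object-relative sphere mask, Fresnel, sunlight and optional rotating-noise caustics.
# Tiled scratches and cracks add detail, then height- and up-vector-masked snow, optionally mixed with gravel, blends on top.
# Switches add a vertex-painted parallax texture and cracked-ice decal before shared and RVT-blending functions.
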Material /BlastBerryMapContent/Game/Athena/WinterArt/Materials/M_Rufus_Ice_Snow
graph {
 properties {
  bUseMaterialAttributes=true
  bUsedWithInstancedStaticMeshes=true
  bUsedWithMeshDeformer=true
  bUsedWithNanite=true
  bUsedWithSkeletalMesh=true
  pixelDepthOffsetMode=PDOM_Legacy
  shadingModel=MSM_Subsurface
  usedShadingModels=Subsurface
 }
 nodes [
  n1 = BumpOffset(
   Coordinate=Multiply(A=TextureCoordinate(), B=ScalarParameter() {defaultValue=5, group=VPTexture, parameterName=VP_TextureScale})
   Height=ScalarParameter() {defaultValue=-1, group=VPTexture, parameterName=Parallax}
  )
  n2 = TextureSampleParameter2D(UVs=@n1) {
   group=VPTexture
   parameterName=D2
   samplerSource=SSM_Wrap_WorldGroupSettings
   texture=/Game/Environments/Sets/IceCastle/Textures/T_IceCastle_Floor1_D
  }
  n3 = TextureSampleParameter2D(UVs=@n1) {
   group=VPTexture
   parameterName=S2
   samplerSource=SSM_Wrap_WorldGroupSettings
   samplerType=SAMPLERTYPE_Masks
   sortPriority=1
   texture=/Game/Environments/Sets/IceCastle/Textures/T_IceCastle_Floor1_S
  }
  n4 = MaterialFunctionCall(Float3=@n3.RGB) {materialFunction=&BreakOutFloat3Components}
  n5 = Constant3Vector() {constant=<rgba 0 0 1 0>}
  n6 = TextureSampleParameter2D(UVs=@n1) {
   group=VPTexture
   parameterName=N2
   samplerSource=SSM_Wrap_WorldGroupSettings
   samplerType=SAMPLERTYPE_Normal
   sortPriority=2
   texture=/Game/Environments/Sets/IceCastle/Textures/T_IceCastle_Floor1_N
  }
  n7 = DotProduct(A=CameraVectorWS(), B=VertexNormalWS())
  n8 = Constant()
  n9 = SetMaterialAttributes!(
   MaterialAttributes
   Base Color=Multiply(A=VectorParameter() {group=VPTexture, parameterName=VPTint, defaultValue=<rgba 1 1 1 0>}.RGB, B=@n2.RGB)
   Metallic=@n4.G
   Specular=@n4.G
   Roughness
   Normal=Transform(None=LinearInterpolate(A=@n5, B=@n6.RGB, Alpha=@n7))
   Emissive Color=@n8
   Subsurface Color=@n8
  ) {attributeSetTypes [
    #1
    #2
    #3
    #4
    #5
    #6
    #7
   ]}
  n10 = TextureObjectParameter() {group=Tile, parameterName=R_Tile, samplerType=SAMPLERTYPE_Masks, texture=/Game/Global/Textures/Default/Tileable_Patterns/T_IceScratch1_M}
  n11 = ScalarParameter() {defaultValue=2, group=Tile, parameterName=TileTextureScale}
  n12 = Divide(A=Constant() {r=1024}, B=@n11)
  n13 = Multiply(A=TextureCoordinate(), B=@n11)
  n14 = StaticSwitchParameter(
   True=ComponentMask(None=MaterialFunctionCall(TextureObject=@n10, TextureSize=@n12) {materialFunction=&WorldAlignedTexture}.XYZ Texture) {r=true}
   False=TextureSample(UVs=@n13, Tex=@n10) {samplerType=SAMPLERTYPE_Masks, texture=&T_IceScratch1_M}.R
  ) {group=Tile, parameterName=Tile WS}
  n15 = MaterialFunctionCall(Transition Phase=VertexColor().G, Height Texture=@n14) {materialFunction=&HeightLerp}
  n16 = VectorParameter() {group=Colors, parameterName=CustomColor, defaultValue=<rgba 0.4 0.0895 0.0376 1>}
  n17 = MaterialFunctionCall() {materialFunction=/BlastBerryMapContent/Game/Athena/Environments/Materials/MaterialFunctions/MF_Rufus_SharedWinter}
  n18 = Fresnel(Normal=VertexNormalWS()) {exponent=1}
  n19 = Constant() {r=5}
  n20 = Constant() {desc=default is .5, r=0.5}
  n21 = StaticSwitchParameter(
   True=VectorParameter() {group=Colors, parameterName=CustomColor, defaultValue=<rgba 0.4 0.0895 0.0376 1>}.RGB
   False=MaterialFunctionCall() {materialFunction=&MF_Rufus_SharedWinter}.IceColor
  ) {group=Colors, parameterName=CustomColor}
  n22 = LinearInterpolate(
   A=ScalarParameter() {defaultValue=-1, group=Tile, parameterName=TileToEmissiveMin}
   B=ScalarParameter() {defaultValue=1, group=Tile, parameterName=TileToEmissiveMax}
   Alpha=@n14
  )
  n23 = TextureSampleParameter2D() {group=Textures, parameterName=D, texture=/Game/Environments/Props/Rocks/Textures/T_Corrupt_Rock_AB_D}
  n24 = LinearInterpolate(A=@n23.A, B=@n23.A, Alpha=ScalarParameter() {group=Textures, parameterName=HeightMap Texture only no Diffuse})
  n25 = MaterialFunctionCall(
   Transition Phase=ScalarParameter() {defaultValue=0.4, group=Textures, parameterName=HeightMapTransition}
   Height Texture=@n24
   Contrast=ScalarParameter() {group=Textures, parameterName=HeightMapContrast}
  ) {materialFunction=&HeightLerp}
  n26 = MaterialFunctionCall(Alpha=@n25.Alpha, Min=Constant() {r=0.0479}, Max=Constant() {r=0.5}) {materialFunction=&SmoothStep}
  n27 = LinearInterpolate(
   A=Desaturation(None=Multiply(A=@n20, B=@n21) {constA=0.5}, Fraction=Constant() {r=-0.7})
   B=@n21
   Alpha=LinearInterpolate(A=Saturate(None=@n22), Alpha=@n26.Result)
  )
  n28 = VertexNormalWS()
  n29 = Saturate(None=Multiply(A=ComponentMask(None=@n28) {b=true}) {constB=-1})
  n30 = ScalarParameter() {defaultValue=3, group=Depth, parameterName=ShpereMax}
  n31 = ScalarParameter() {defaultValue=-9, group=Depth, parameterName=SphereMin}
  n32 = SphereMask(
   A=Subtract(A=WorldPosition().XYZ, B=ActorPositionWS())
   B=VectorParameter() {group=Depth, parameterName=SphereLocation, defaultValue=<rgba 0 0 0 1>}.RGB
   Radius=Multiply(
    A=ObjectRadius()
    B=ScalarParameter() {defaultValue=1, desc=works automatically when scaling mesh, group=Depth, parameterName="SphereRadius "}
   )
   Hardness=ScalarParameter() {defaultValue=-3.5, group=Depth, parameterName=SphereHardness}
  )
  n33 = LinearInterpolate(A=Constant() {r=0.4}, B=Constant() {r=-1.5456}, Alpha=@n32)
  n34 = Multiply(A=AtmosphericLightVector()) {constB=-1}
  n35 = CameraVectorWS()
  n36 = Power(Base=Clamp(None=DotProduct(A=@n34, B=@n35)), Exp=ScalarParameter() {defaultValue=8, parameterName=AtmosphereExponent})
  n37 = VectorParameter() {group=Colors, parameterName="Atmospheric Light Tint ", defaultValue=<rgba 5.9913 8.3586 10 0>}
  n38 = Add(
   A=Clamp(None=LinearInterpolate(A=@n30, B=@n31, Alpha=@n32)) {maxDefault=5}
   B=Multiply(A=Clamp(None=@n33) {maxDefault=5}, B=Multiply(A=@n36, B=@n37.RGB))
  )
  n39 = ScalarParameter() {group=Caustics, parameterName=MinCaustic}
  n40 = ComponentMask(None=VertexNormalWS()) {b=true}
  n41 = WorldPosition()
  n42 = ActorPositionWS()
  n43 = Divide(
   A=ComponentMask(None=Subtract(A=@n41.XYZ, B=@n42)) {g=true, r=true}
   B=ScalarParameter() {defaultValue=7050.1357, group=Caustics, parameterName=ScaleNoiseInside}
  )
  n44 = Add(
   A=Multiply(A=ScalarParameter() {defaultValue=0.125, group=Caustics, parameterName=CausticIdleSpeed}, B=Time()) {constA=0.125}
   B=ComponentMask(None=CameraVectorWS()) {r=true}
  )
  n45 = TextureSampleParameter2D(UVs=Rotator(Coordinate=@n43, Time=@n44)) {group=Caustics, parameterName=Caustics, samplerType=SAMPLERTYPE_LinearColor, texture=/Game/Global/Textures/Default/Noise/T_Noise01}
  n46 = TextureSampleParameter2D(UVs=Rotator(Coordinate=@n43, Time=@n44) {speed=-0.25}) {group=Caustics, parameterName=Caustics, samplerType=SAMPLERTYPE_LinearColor, texture=&T_Noise01}
  n47 = Fresnel(Normal=VertexNormalWS()) {exponent=1}
  n48 = LinearInterpolate(
   B=LinearInterpolate(A=Add(A=@n45.R, B=@n46.R), Alpha=Abs(None=@n40))
   Alpha=MaterialFunctionCall(Alpha=OneMinus(None=@n47), Max=Constant() {r=2}) {materialFunction=&SmoothStep}.Result
  ) {constA=1}
  n49 = Multiply(A=@n38, B=Saturate(None=LinearInterpolate(A=@n39, Alpha=@n48)))
  n50 = ScalarParameter() {defaultValue=0.2, group=Depth, parameterName=FresnelMin}
  n51 = ScalarParameter() {defaultValue=1, group=Depth, parameterName=FresnelMax}
  n52 = VertexNormalWS()
  n53 = MaterialFunctionCall(Alpha=OneMinus(None=Fresnel(Normal=@n52) {exponent=1}), Max=Constant() {r=2}) {materialFunction=&SmoothStep}
  n54 = Multiply(
   A=OneMinus(None=@n25.Alpha)
   B=Multiply(A=StaticSwitchParameter(True=@n49, False=@n38) {parameterName=CausticEffect True}, B=LinearInterpolate(A=@n50, B=@n51, Alpha=@n53.Result))
  )
  n55 = StaticSwitchParameter(True=Multiply(A=LinearInterpolate(Alpha=@n29) {constA=1, constB=0.2}, B=@n54), False=@n54) {parameterName=DownVector Shadow}
  n56 = TextureObjectParameter() {group=Tile, parameterName=N_Tile, samplerType=SAMPLERTYPE_Normal, texture=/Game/Global/Textures/Default/Tileable_Patterns/T_IceScratch1_N}
  n57 = MaterialFunctionCall(TextureObject=@n56, TextureSize=@n12) {materialFunction=&WorldAlignedNormal}
  n58 = TextureSample(UVs=@n13, Tex=@n56) {samplerType=SAMPLERTYPE_Normal, texture=&T_IceScratch1_N}
  n59 = Constant3Vector() {constant=<rgba 0 0 1 0>}
  n60 = ComponentMask(None=VertexNormalWS()) {b=true}
  n61 = LinearInterpolate(
   A=LinearInterpolate(A=StaticSwitchParameter(True=@n57.XYZ Texture, False=@n58.RGB) {group=Tile, parameterName=Tile WS}, B=@n59, Alpha=Abs(None=@n60))
   B=@n59
   Alpha=@n26.Result
  )
  n62 = ScalarParameter() {defaultValue=0.85, group=Tile, parameterName=Tile N Flatten}
  n63 = TextureSampleParameter2D() {group=Textures, parameterName=N, samplerType=SAMPLERTYPE_Normal, texture=/Game/Environments/Props/Rocks/Textures/T_Corrupt_Rock_AB_N}
  n64 = Constant() {r=0.2}
  n65 = ScalarParameter() {defaultValue=0.25, group=SSS, parameterName=SSSMin}
  n66 = ScalarParameter() {defaultValue=1, group=SSS, parameterName=SSSMax}
  n67 = SetMaterialAttributes!(
   MaterialAttributes
   Base Color=Multiply(
    A=StaticSwitchParameter(True=@n16.RGB, False=@n17.IceColor) {group=Colors, parameterName=CustomColor}
    B=MaterialFunctionCall(Alpha=@n18, Max=@n19) {materialFunction=&SmoothStep}.Result
   )
   Specular=Constant() {r=0.25}
   Roughness=LinearInterpolate(A=Multiply(A=@n14, B=@n64), Alpha=@n26.Result) {constB=0.025}
   Emissive Color=Divide(A=Multiply(A=@n27, B=@n55), B=EyeAdaptation().EyeAdaptation)
   Normal=MaterialFunctionCall(
    BaseNormal=@n63.RGB
    AdditionalNormal=MaterialFunctionCall(Normal=@n61, Flatness=@n62) {materialFunction=&FlattenNormal}.Result
   ) {materialFunction=&BlendAngleCorrectedNormals}.Result
   Subsurface Color=Multiply(A=LinearInterpolate(A=@n65, B=@n66, Alpha=@n25.Alpha), B=@n21)
  ) {attributeSetTypes [
    #1
    #3
    #4
    #6
    #5
    #7
   ]}
  n68 = MaterialFunctionCall!(In=@n67) {materialFunction=/Game/Athena/WinterArt/Materials/MF_Cracks}
  n69 = MaterialFunctionCall!(UseLandscapeCoordsOnSwitch) {materialFunction=/BlastBerryMapContent/Game/Athena/Environments/Materials/MaterialFunctions/MF_Athena_Snow_01}
  n70 = WorldPosition()
  n71 = MaterialFunctionCall!(
   Coords=Divide(A=ComponentMask(None=@n70.XYZ) {g=true, r=true}, B=ScalarParameter() {defaultValue=512, group=Gravel, parameterName=VP_TexSize})
   Color Tint=ComponentMask(None=CollectionParameter() {collection=/Game/Athena/Environments/Landscape/MPC/MPC_Landscape, parameterName=Winter_Gravel_Tint}) {b=true, g=true, r=true}
  ) {materialFunction=/Game/Athena/Environments/Landscape/MaterialFunctions/Standard/MF_Athena_Gravel_01}
  n72 = MaterialFunctionCall(Transition Phase=OneMinus(None=VertexColor().A), Height Texture=@n71.Height, Contrast=Constant() {r=1}) {materialFunction=&HeightLerp}
  n73 = GetMaterialAttributes(None=StaticSwitchParameter(True=BlendMaterialAttributes(A=@n69.Material, B=@n71.Result, Alpha=@n72.Alpha), False=@n69.Material) {group=Gravel, parameterName=UseVP_OnSnow}) {attributeGetTypes [#5, #1]}
  n74 = Constant() {r=1}
  n75 = SetMaterialAttributes!(
   MaterialAttributes=@n73.MaterialAttributes
   Normal=@n73.MaterialAttributes
   Emissive Color=Constant()
   Subsurface Color=NamedRerouteUsage()
   Base Color=@n73.MaterialAttributes
   Opacity=FeatureLevelSwitch(Default=Constant() {r=0.95}, ES2=@n74, ES3_1=@n74)
  ) {attributeSetTypes [
    #5
    #6
    #7
    #1
    #8
   ]}
  n76 = LinearInterpolate(
   A=VertexColor().R
   B=ScalarParameter() {defaultValue=0.4, parameterName=Snow Transition Phase}
   Alpha=ScalarParameter() {parameterName=Snow Uses Height}
  )
  n77 = ScalarParameter() {parameterName=Snow Contrast}
  n78 = Multiply(
   A=ComponentMask(None=Transform(None=@n63.RGB)) {b=true}
   B=Multiply(
    A=MaterialFunctionCall(Transition Phase=@n76, Height Texture=@n24, Contrast=@n77) {materialFunction=&HeightLerp}.Alpha
    B=ScalarParameter() {defaultValue=1, group=Colors, parameterName=SnowUpVector}
   )
  )
  n79 = StaticSwitchParameter(
   True=MaterialFunctionCall!(Base Material=@n68.Result, Top Material=@n75, Alpha=Saturate(None=@n78)) {materialFunction=/Engine/Functions/MaterialLayerFunctions/MatLayerBlend_Standard}.Blended Material
   False=@n68.Result
  ) {defaultValue=true, parameterName=Snow}
  n80 = StaticSwitchParameter(
   True=MaterialFunctionCall(Base Material=@n79, Top Material=@n9, Alpha=Multiply(A=@n15.Alpha, B=@n7)) {materialFunction=&MatLayerBlend_Standard}.Blended Material
   False=@n79
  ) {group=VPTexture, parameterName=VPTexture}
  n81 = StaticSwitchParameter(True=MaterialFunctionCall!(In=@n80) {materialFunction=/Game/Athena/WinterArt/Materials/MF_DecalIceCracked}.Result, False=@n80) {parameterName=UseCrackedIceDecal}
  MaterialFunctionCall!(In=@n81) {materialFunction=/Game/Athena/WinterArt/Materials/MF_GameWideShared}
  MaterialFunctionCall!(In=@n81) {materialFunction=/BlastBerryMapContent/Game/TechArt/Material_Functions/MF_Rufus_RVTBlendingOn}
  n82 = VectorParameter() {group=PostRVTSnow, parameterName=SnowSSSColor, defaultValue=<rgba 0.0677 0.2393 0.5 1>}
  n83 = CameraVectorWS()
  n84 = PixelNormalWS()
  n85 = OneMinus(None=Saturate(None=DotProduct(A=@n83, B=@n84)))
  n86 = ScalarParameter() {defaultValue=4, group=PostRVTSnow, parameterName=SnowSpecularSelfShadowing}
  n87 = LinearInterpolate(
   A=ScalarParameter() {defaultValue=0.5, group=PostRVTSnow, parameterName=SnowSSSFacingExponent}
   Alpha=Saturate(None=Power(Base=@n85, Exp=@n86) {constExponent=3})
  )
  n88 = Constant()
  NamedRerouteDeclaration(None=FeatureLevelSwitch(Default=Multiply(A=@n82.RGB, B=@n87), ES2=@n88, ES3_1=@n88)) {name=snow subsurface color, nodeColor=<rgba 0.4766 0 1 1>}
  StaticSwitchParameter(
   True=MaterialFunctionCall!(BaseColor=@n73.Base Color) {materialFunction=/Game/Athena/Environments/Landscape/MaterialFunctions/Standard/MF_Athena_GrassMatch}.Result
   False=@n73.Base Color
  ) {defaultValue=true, parameterName=UseColorMap}
 ]
}
